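# Sound cue for a Section 01 spire in the PistolCanary event, one numbered variant that plays a spire energy ignite sound.
# A single wave player feeds the output directly, reusing a Spire_Ignite_Energy wave from the Section00 folder.
# It uses the spire activation attenuation with a long max distance, belongs to the Fort_Events sound class and plays at reduced volume.
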SoundCue /PistolCanary_Event_Audio/Assets/Section01/Spires/SC_PistolCanary_Section01_Spires_03
root=WavePlayer_0
nodes [{id=WavePlayer_0, properties {soundWaveAssetPtr=/PistolCanary_Event_Audio/Assets/Section00/wav/PistolCanary_Spire_Ignite_Energy_03}}]
properties {
 attenuationSettings=/PistolCanary_Event_Audio/Assets/Settings/ATT_PistolCanary_Spires_Activate
 maxDistance=28000
 soundClassObject=&Fort_Events
 volumeMultiplier=0.5
}
