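# Looping SoundCue for the spire beam in Section01 of the PistolCanary event, classed under Fort_Events.
# A looping wave player, using a beam loop wave from the Section00 folder, feeds an Enveloper at the root.
# The envelope quickly fades volume in to full, then eases it down to half.
# The cue plays at reduced volume with shared spire ambient-loop attenuation.
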SoundCue /PistolCanary_Event_Audio/Assets/Section01/Spires/SC_PistolCanary_Section01_Spires_BeamLoop_01
root=Enveloper_0
nodes [
 {
  id=Enveloper_0
  children [WavePlayer_4]
  properties {pitchCurve {editorCurveData {keys [{time=0}]}}, volumeCurve {editorCurveData {keys [<at 0 0>,<at 0.2 1>,<at 2.5 0.5>]}}}
 }
 {
  id=WavePlayer_4
  properties {bLooping=true, soundWaveAssetPtr=/PistolCanary_Event_Audio/Assets/Section00/wav/PistolCanary_Spire_Beam_Loop_01}
 }
]
properties {
 attenuationSettings=/PistolCanary_Event_Audio/Assets/Settings/ATT_PistolCanary_Spires_AmbLoops
 maxDistance=36000
 soundClassObject=&Fort_Events
 volumeMultiplier=0.25
}
